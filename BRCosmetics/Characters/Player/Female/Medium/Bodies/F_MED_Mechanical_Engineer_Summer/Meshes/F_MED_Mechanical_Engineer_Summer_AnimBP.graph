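# Post-copy AnimBlueprint for the Mechanical Engineer Summer medium female body, positioning its eleven dyn_pantBall bones relative to the body.
# Copy Pose From Mesh takes the attached parent's pose, then a chain of bone Constraint nodes runs in component space.
# Balls 1-2 and 8-11 follow the pelvis alone; 3-7 blend pelvis, thighs and spine_01 with fixed, pelvis-dominant weights.
# It has no simulation, variables or event logic.
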AnimBlueprint /BRCosmetics/Characters/Player/Female/Medium/Bodies/F_MED_Mechanical_Engineer_Summer/Meshes/F_MED_Mechanical_Engineer_Summer_AnimBP
target_skeleton=&/F_MED_Mechanical_Engineer_Summer_Skeleton
graphs [
 AnimGraph {nodes [
   n1 = "Constraint - dyn_pantBall_1 (P)"!(
    ComponentPose=Local To Component(LocalPose=Copy Pose From Mesh() {bUseAttachedParent=true})
    Alpha=1
    ConstraintWeights_0=1
    ConstraintWeights_1:float=0
   ) -> (Pose) {
    alphaBoolBlend {blendOption=Linear}
    boneToModify {boneName=dyn_pantBall_1}
    constraintSetup [{targetBone {boneName=pelvis}}]
    constraintWeights [1,0]
   }
   n2 = "Constraint - dyn_pantBall_2 (P)"!(ComponentPose=@n1, Alpha=1, ConstraintWeights_0=1, ConstraintWeights_1:float=0) -> (Pose) {
    alphaBoolBlend {blendOption=Linear}
    boneToModify {boneName=dyn_pantBall_2}
    constraintSetup [{targetBone {boneName=pelvis}}]
    constraintWeights [1,0]
   }
   n3 = "Constraint - dyn_pantBall_3 (PPPP)"!(
    ComponentPose=@n2
    Alpha=1
    ConstraintWeights_0=0
    ConstraintWeights_1:float=0.0003
    ConstraintWeights_2:float=0.0051
    ConstraintWeights_3:float=0.9946
   ) -> (Pose) {
    alphaBoolBlend {blendOption=Linear}
    boneToModify {boneName=dyn_pantBall_3}
    constraintSetup [{targetBone {boneName=thigh_r}}, {targetBone {boneName=spine_01}}, {targetBone {boneName=thigh_l}}, {targetBone {boneName=pelvis}}]
    constraintWeights [1,1,1,1]
   }
   n4 = "Constraint - dyn_pantBall_4 (PPPP)"!(
    ComponentPose=@n3
    Alpha=1
    ConstraintWeights_0=0.0002
    ConstraintWeights_1:float=0.0011
    ConstraintWeights_2:float=0.0124
    ConstraintWeights_3:float=0.9864
   ) -> (Pose) {
    alphaBoolBlend {blendOption=Linear}
    boneToModify {boneName=dyn_pantBall_4}
    constraintSetup [{targetBone {boneName=thigh_r}}, {targetBone {boneName=spine_01}}, {targetBone {boneName=thigh_l}}, {targetBone {boneName=pelvis}}]
    constraintWeights [1,1,1,1]
   }
   n5 = "Constraint - dyn_pantBall_5 (PPPP)"!(
    ComponentPose=@n4
    Alpha=1
    ConstraintWeights_0=0.0024
    ConstraintWeights_1:float=0.0083
    ConstraintWeights_2:float=0.0002
    ConstraintWeights_3:float=0.9891
   ) -> (Pose) {
    alphaBoolBlend {blendOption=Linear}
    boneToModify {boneName=dyn_pantBall_5}
    constraintSetup [{targetBone {boneName=thigh_r}}, {targetBone {boneName=spine_01}}, {targetBone {boneName=thigh_l}}, {targetBone {boneName=pelvis}}]
    constraintWeights [1,1,1,1]
   }
   n6 = "Constraint - dyn_pantBall_6 (PPPP)"!(
    ComponentPose=@n5
    Alpha=1
    ConstraintWeights_0=0.0124
    ConstraintWeights_1:float=0.0011
    ConstraintWeights_2:float=0.0002
    ConstraintWeights_3:float=0.9864
   ) -> (Pose) {
    alphaBoolBlend {blendOption=Linear}
    boneToModify {boneName=dyn_pantBall_6}
    constraintSetup [{targetBone {boneName=thigh_r}}, {targetBone {boneName=spine_01}}, {targetBone {boneName=thigh_l}}, {targetBone {boneName=pelvis}}]
    constraintWeights [1,1,1,1]
   }
   n7 = "Constraint - dyn_pantBall_7 (PPPP)"!(
    ComponentPose=@n6
    Alpha=1
    ConstraintWeights_0=0.0051
    ConstraintWeights_1:float=0.0003
    ConstraintWeights_2:float=0
    ConstraintWeights_3:float=0.9946
   ) -> (Pose) {
    alphaBoolBlend {blendOption=Linear}
    boneToModify {boneName=dyn_pantBall_7}
    constraintSetup [{targetBone {boneName=thigh_r}}, {targetBone {boneName=spine_01}}, {targetBone {boneName=thigh_l}}, {targetBone {boneName=pelvis}}]
    constraintWeights [1,1,1,1]
   }
   n8 = "Constraint - dyn_pantBall_8 (P)"!(ComponentPose=@n7, Alpha=1, ConstraintWeights_0=1, ConstraintWeights_1:float=0) -> (Pose) {
    alphaBoolBlend {blendOption=Linear}
    boneToModify {boneName=dyn_pantBall_8}
    constraintSetup [{targetBone {boneName=pelvis}}]
    constraintWeights [1,0]
   }
   n9 = "Constraint - dyn_pantBall_9 (P)"!(ComponentPose=@n8, Alpha=1, ConstraintWeights_0=1, ConstraintWeights_1:float=0) -> (Pose) {
    alphaBoolBlend {blendOption=Linear}
    boneToModify {boneName=dyn_pantBall_9}
    constraintSetup [{targetBone {boneName=pelvis}}]
    constraintWeights [1,0]
   }
   n10 = "Constraint - dyn_pantBall_10 (P)"!(ComponentPose=@n9, Alpha=1, ConstraintWeights_0=1, ConstraintWeights_1:float=0) -> (Pose) {
    alphaBoolBlend {blendOption=Linear}
    boneToModify {boneName=dyn_pantBall_10}
    constraintSetup [{targetBone {boneName=pelvis}}]
    constraintWeights [1,0]
   }
   n11 = "Constraint - dyn_pantBall_11 (P)"!(ComponentPose=@n10, Alpha=1, ConstraintWeights_0=1, ConstraintWeights_1:float=0) -> (Pose) {
    alphaBoolBlend {blendOption=Linear}
    boneToModify {boneName=dyn_pantBall_11}
    constraintSetup [{targetBone {boneName=pelvis}}]
    constraintWeights [1,0]
   }
   Output Pose(Result=Component To Local(ComponentPose=@n11)) {layerGroup=None}
  ]}
 EventGraph {nodes [Event BlueprintUpdateAnimation()]}
]
variables []
