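# Physics asset for the Summer Mechanical Engineer female character, covering dynamic hair and the upper body it can collide with.
# Left and right six-link tails, three-link front tufts and a single upper hair piece form angle-limited capsule chains on small base bodies.
# Those bases join the head with free angular constraints.
# Capsules cover head, upper spine, clavicles and upper arms, while a large offset box sits on neck_02, angularly locked to the head.
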PhysicsAsset /BRCosmetics/Characters/Player/Female/Medium/Bodies/F_MED_Mechanical_Engineer_Summer/Meshes/Parts/F_MED_Mechanical_Engineer_Summer_Physics
bodies {
 clavicle_l [{shapeType=Capsule, length=24.625, radius=4.375, center=<4.507 -0.0001 -0.7571>, rotation=<rot 82.9999 0 0>}]
 clavicle_r [{shapeType=Capsule, length=24.75, radius=4.25, center=<-4.5408 0 0.4821>, rotation=<rot 82.9999 0 0>}]
 dyn_hair_fr_tuft_a_base_l [{shapeType=Capsule, length=0.505, radius=0.505, rotation=<rot 0 -90 0>}]
 dyn_hair_fr_tuft_a_base_r [{shapeType=Capsule, length=0.505, radius=0.505, rotation=<rot 0 -90 0>}]
 dyn_hair_fr_tuft_a_l [{shapeType=Capsule, length=1.75, radius=1, center=<1.8829 0 0>, rotation=<rot 90 0 0>}]
 dyn_hair_fr_tuft_a_r [{shapeType=Capsule, length=1.75, radius=1, center=<-1.8829 0 0>, rotation=<rot 90 0 0>}]
 dyn_hair_fr_tuft_b_l [{shapeType=Capsule, length=1, radius=1, center=<1.3448 0 0>, rotation=<rot 90 0 0>}]
 dyn_hair_fr_tuft_b_r [{shapeType=Capsule, length=1, radius=1, center=<-1.8829 0 0>, rotation=<rot 90 0 0>}]
 dyn_hair_fr_tuft_c_l [{shapeType=Capsule, length=1.75, radius=1, center=<1.8829 0 0>, rotation=<rot 90 0 0>}]
 dyn_hair_fr_tuft_c_r [{shapeType=Capsule, length=1.75, radius=1, center=<-1.8829 0 0>, rotation=<rot 90 0 0>}]
 dyn_hair_tail_a_base_l [{shapeType=Capsule, length=0.505, radius=0.505, rotation=<rot 0 -90 0>}]
 dyn_hair_tail_a_base_r [{shapeType=Capsule, length=0.505, radius=0.505, rotation=<rot 0 -90 0>}]
 dyn_hair_tail_a_l [{shapeType=Capsule, length=1.25, radius=2.25, center=<2.3026 0.9413 0>, rotation=<rot -90 0 7.0306>}]
 dyn_hair_tail_a_r [{shapeType=Capsule, length=1.25, radius=2.25, center=<-2.34 -0.3766 0>, rotation=<rot -90 0 0>}]
 dyn_hair_tail_b_l [{shapeType=Capsule, length=3.75, radius=1.75, center=<2.5 0 0>, rotation=<rot 90 0 0>}]
 dyn_hair_tail_b_r [{shapeType=Capsule, length=3.75, radius=1.75, center=<-2.5 0 0>, rotation=<rot 90 0 0>}]
 dyn_hair_tail_c_l [{shapeType=Capsule, length=3.75, radius=1.5, center=<2.5 0 0>, rotation=<rot 90 0 0>}]
 dyn_hair_tail_c_r [{shapeType=Capsule, length=3.75, radius=1.5, center=<-2.5 0 0>, rotation=<rot 90 0 0>}]
 dyn_hair_tail_d_l [{shapeType=Capsule, length=3.75, radius=1.75, center=<2.5 0 0>, rotation=<rot 90 0 0>}]
 dyn_hair_tail_d_r [{shapeType=Capsule, length=3.75, radius=1.75, center=<-2.5 0 0>, rotation=<rot 90 0 0>}]
 dyn_hair_tail_e_l [{shapeType=Capsule, length=3.75, radius=1.3, center=<2.5 0 0>, rotation=<rot 90 0 0>}]
 dyn_hair_tail_e_r [{shapeType=Capsule, length=3.75, radius=1.3, center=<-2.5 0 0>, rotation=<rot 90 0 0>}]
 dyn_hair_tail_f_l [{shapeType=Capsule, length=5.5, radius=1.5, center=<4.4 0 0>, rotation=<rot 90 0 0>}]
 dyn_hair_tail_f_r [{shapeType=Capsule, length=5.5, radius=1.5, center=<-4.4 0 0>, rotation=<rot 90 0 0>}]
 dyn_hair_up [{shapeType=Capsule, length=1, radius=1, center=<-0.5 0 0>, rotation=<rot 90 0 0>}]
 dyn_hair_up_base [{shapeType=Capsule, length=0.505, radius=0.505, rotation=<rot 0 -90 0>}]
 head [{shapeType=Capsule, length=8.25, radius=7, center=<3 1.1004 0>, rotation=<rot 90 0 0>}]
 neck_02 [{shapeType=Box, extentX=35, extentY=24.5, extentZ=56.51, center=<-4.645 -14.421 -0.7955>, rotation=<rot 0 -12.2778 0>}]
 spine_05 [{shapeType=Capsule, length=10, radius=8.75, center=<-3.1411 2.9137 0>, rotation=<rot 0 -90 0>}]
 upperarm_l [{shapeType=Capsule, length=18, radius=4.9, center=<9.8305 0 0>, rotation=<rot 90 0 0>}]
 upperarm_r [{shapeType=Capsule, length=18, radius=6, center=<-11 0 0>, rotation=<rot 90 0 0>}]
}
constraints [
 {
  bone1Name=head
  bone2Name=neck_02
  swing1LimitDegrees=45
  swing1Motion=Locked
  swing2LimitDegrees=45
  swing2Motion=Locked
  twistLimitDegrees=45
  twistMotion=Locked
 }
 {
  bone1Name=dyn_hair_tail_a_base_r
  bone2Name=head
  swing1LimitDegrees=45
  swing1Motion=Free
  swing2LimitDegrees=45
  swing2Motion=Free
  twistLimitDegrees=45
  twistMotion=Free
 }
 {bone1Name=dyn_hair_tail_a_r, bone2Name=dyn_hair_tail_a_base_r, swing1LimitDegrees=70, swing2LimitDegrees=20, twistLimitDegrees=5}
 {bone1Name=dyn_hair_tail_b_r, bone2Name=dyn_hair_tail_a_r, swing1LimitDegrees=70, swing2LimitDegrees=40, twistLimitDegrees=5}
 {bone1Name=dyn_hair_tail_c_r, bone2Name=dyn_hair_tail_b_r, swing1LimitDegrees=60, swing2LimitDegrees=30, twistLimitDegrees=5}
 {bone1Name=dyn_hair_tail_d_r, bone2Name=dyn_hair_tail_c_r, swing1LimitDegrees=60, swing2LimitDegrees=30, twistLimitDegrees=5}
 {bone1Name=dyn_hair_tail_e_r, bone2Name=dyn_hair_tail_d_r, swing1LimitDegrees=60, swing2LimitDegrees=30, twistLimitDegrees=5}
 {bone1Name=dyn_hair_tail_f_r, bone2Name=dyn_hair_tail_e_r, swing1LimitDegrees=60, swing2LimitDegrees=30, twistLimitDegrees=5}
 {
  bone1Name=dyn_hair_tail_a_base_l
  bone2Name=head
  swing1LimitDegrees=45
  swing1Motion=Free
  swing2LimitDegrees=45
  swing2Motion=Free
  twistLimitDegrees=45
  twistMotion=Free
 }
 {bone1Name=dyn_hair_tail_a_l, bone2Name=dyn_hair_tail_a_base_l, swing1LimitDegrees=70, swing2LimitDegrees=20, twistLimitDegrees=5}
 {bone1Name=dyn_hair_tail_b_l, bone2Name=dyn_hair_tail_a_l, swing1LimitDegrees=70, swing2LimitDegrees=40, twistLimitDegrees=5}
 {bone1Name=dyn_hair_tail_c_l, bone2Name=dyn_hair_tail_b_l, swing1LimitDegrees=60, swing2LimitDegrees=30, twistLimitDegrees=5}
 {bone1Name=dyn_hair_tail_d_l, bone2Name=dyn_hair_tail_c_l, swing1LimitDegrees=60, swing2LimitDegrees=30, twistLimitDegrees=5}
 {bone1Name=dyn_hair_tail_e_l, bone2Name=dyn_hair_tail_d_l, swing1LimitDegrees=60, swing2LimitDegrees=30, twistLimitDegrees=5}
 {bone1Name=dyn_hair_tail_f_l, bone2Name=dyn_hair_tail_e_l, swing1LimitDegrees=60, swing2LimitDegrees=30, twistLimitDegrees=5}
 {
  bone1Name=dyn_hair_up_base
  bone2Name=head
  swing1LimitDegrees=45
  swing1Motion=Free
  swing2LimitDegrees=45
  swing2Motion=Free
  twistLimitDegrees=45
  twistMotion=Free
 }
 {bone1Name=dyn_hair_up, bone2Name=dyn_hair_up_base, swing1LimitDegrees=15, swing2LimitDegrees=15, twistLimitDegrees=15}
 {
  bone1Name=dyn_hair_fr_tuft_a_base_r
  bone2Name=head
  swing1LimitDegrees=45
  swing1Motion=Free
  swing2LimitDegrees=45
  swing2Motion=Free
  twistLimitDegrees=45
  twistMotion=Free
 }
 {bone1Name=dyn_hair_fr_tuft_a_r, bone2Name=dyn_hair_fr_tuft_a_base_r, swing1LimitDegrees=5, swing2LimitDegrees=10, twistLimitDegrees=5}
 {bone1Name=dyn_hair_fr_tuft_b_r, bone2Name=dyn_hair_fr_tuft_a_r, swing1LimitDegrees=10, swing2LimitDegrees=10, twistLimitDegrees=5}
 {bone1Name=dyn_hair_fr_tuft_c_r, bone2Name=dyn_hair_fr_tuft_b_r, swing1LimitDegrees=15, swing2LimitDegrees=10, twistLimitDegrees=5}
 {
  bone1Name=dyn_hair_fr_tuft_a_base_l
  bone2Name=head
  swing1LimitDegrees=45
  swing1Motion=Free
  swing2LimitDegrees=45
  swing2Motion=Free
  twistLimitDegrees=45
  twistMotion=Free
 }
 {bone1Name=dyn_hair_fr_tuft_a_l, bone2Name=dyn_hair_fr_tuft_a_base_l, swing1LimitDegrees=5, swing2LimitDegrees=10, twistLimitDegrees=5}
 {bone1Name=dyn_hair_fr_tuft_b_l, bone2Name=dyn_hair_fr_tuft_a_l, swing1LimitDegrees=10, swing2LimitDegrees=10, twistLimitDegrees=5}
 {bone1Name=dyn_hair_fr_tuft_c_l, bone2Name=dyn_hair_fr_tuft_b_l, swing1LimitDegrees=15, swing2LimitDegrees=10, twistLimitDegrees=5}
 {
  bone1Name=clavicle_l
  bone2Name=spine_05
  swing1LimitDegrees=45
  swing1Motion=Free
  swing2LimitDegrees=45
  swing2Motion=Free
  twistLimitDegrees=45
  twistMotion=Free
 }
 {
  bone1Name=clavicle_r
  bone2Name=spine_05
  swing1LimitDegrees=45
  swing1Motion=Free
  swing2LimitDegrees=45
  swing2Motion=Free
  twistLimitDegrees=45
  twistMotion=Free
 }
 {
  bone1Name=upperarm_r
  bone2Name=clavicle_r
  swing1LimitDegrees=45
  swing1Motion=Free
  swing2LimitDegrees=45
  swing2Motion=Free
  twistLimitDegrees=45
  twistMotion=Free
 }
 {
  bone1Name=upperarm_l
  bone2Name=clavicle_l
  swing1LimitDegrees=45
  swing1Motion=Free
  swing2LimitDegrees=45
  swing2Motion=Free
  twistLimitDegrees=45
  twistMotion=Free
 }
 {
  bone1Name=neck_02
  bone2Name=spine_05
  swing1LimitDegrees=45
  swing1Motion=Free
  swing2LimitDegrees=45
  swing2Motion=Free
  twistLimitDegrees=45
  twistMotion=Free
 }
]
