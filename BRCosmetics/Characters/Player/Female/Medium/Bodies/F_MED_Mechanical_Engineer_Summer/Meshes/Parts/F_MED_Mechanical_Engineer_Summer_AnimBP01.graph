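# Animation Blueprint for a part mesh of the Mechanical Engineer Summer female player character, posing its hair tails and helmet.
# It copies the attached parent's pose, layers the body mesh's upper arms over it, then applies additive hair offsets and emote constraints.
# Range-mapped HeadUp and HeadTwist drive hair correctives, and HelmetUp rotates the helmet before a spine_05-space RigidBody simulation.
# A baked hair offset applies on mobile or PC performance mode when rigid body is disabled.
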AnimBlueprint /BRCosmetics/Characters/Player/Female/Medium/Bodies/F_MED_Mechanical_Engineer_Summer/Meshes/Parts/F_MED_Mechanical_Engineer_Summer_AnimBP01
target_skeleton=/Game/Characters/Player/Female/Medium/Bodies/F_MED_Mechanical_Engineer_Summer/Meshes/Parts/F_MED_Mechanical_Engineer_Summer_Skeleton
graphs [
 AnimGraph {nodes [
   n1 = Get bIsTargeting()
   n2 = "Transform (Modify) Bone - Bone: dyn_hair_tail_a_base_l"!(ComponentPose, bAlphaBoolEnabled:bool=@n1, Translation="0,0,1", Rotation="-5,0,0", Scale="1,1,1") -> (Pose) {
    alphaInputType=Bool
    rotationMode=BMM_Additive
    rotationSpace=BCS_BoneSpace
    translationMode=BMM_Additive
    translationSpace=BCS_BoneSpace
    alphaBoolBlend {blendOption=Linear}
    alphaScaleBiasClamp {inRange {max=0, min=-50}, outRange {max=0, min=1}}
    boneToModify {boneName=dyn_hair_tail_a_base_l}
    rotation {pitch=-5}
    translation {z=1}
   }
   "Transform (Modify) Bone - Bone: dyn_hair_tail_a_base_r"!(ComponentPose=@n2, bAlphaBoolEnabled:bool=@n1, Translation="0,0,0", Rotation="-15,0,0", Scale="1,1,1") -> (Pose) {
    alphaInputType=Bool
    rotationMode=BMM_Additive
    rotationSpace=BCS_BoneSpace
    translationMode=BMM_Additive
    translationSpace=BCS_BoneSpace
    alphaBoolBlend {blendInTime=0.1, blendOption=Linear, blendOutTime=0.1}
    alphaScaleBiasClamp {inRange {max=0, min=-50}, outRange {max=0, min=1}}
    boneToModify {boneName=dyn_hair_tail_a_base_r}
    rotation {pitch=-15}
   }
   n3 = Copy Pose From Mesh(SourceMeshComponent=Get Body_Mesh())
   n4 = Copy Pose From Mesh() {bCopyCurves=true, bUseAttachedParent=true}
   n5 = Get instead_bake_pose() -> (instead_bake_pose:bool)
   n6 = "Transform (Modify) Bone - Bone: dyn_hair_tail_a_base_l"(
    ComponentPose=Local To Component(LocalPose=Layered blend per bone(BasePose=@n4, BlendPoses_0=@n3) {layerSetup [{branchFilters [{boneName=upperarm_l}, {boneName=upperarm_r}]}]})
    bAlphaBoolEnabled=@n5
    Translation="0,0,-1"
    Rotation="-4.472,-5.355,-1.411"
    Scale="1,1,1"
   ) {
    alphaInputType=Bool
    rotationMode=BMM_Additive
    rotationSpace=BCS_BoneSpace
    translationMode=BMM_Additive
    translationSpace=BCS_BoneSpace
    alphaBoolBlend {blendOption=Linear}
    alphaScaleBiasClamp {inRange {max=0, min=-50}, outRange {max=0, min=1}}
    boneToModify {boneName=dyn_hair_tail_a_base_l}
    rotation=<rot -4.472 -5.3549 -1.411>
    translation {z=-1}
   }
   n7 = "Transform (Modify) Bone - Bone: dyn_hair_tail_a_base_r"(ComponentPose=@n6, bAlphaBoolEnabled=@n5, Translation="0.246,-0.317,3.812", Rotation="0.091,-9.464,3.379", Scale="1,1,1") {
    alphaInputType=Bool
    rotationMode=BMM_Additive
    rotationSpace=BCS_BoneSpace
    translationMode=BMM_Additive
    translationSpace=BCS_BoneSpace
    alphaBoolBlend {blendOption=Linear}
    alphaScaleBiasClamp {inRange {max=0, min=-50}, outRange {max=0, min=1}}
    boneToModify {boneName=dyn_hair_tail_a_base_r}
    rotation=<rot 0.091 -9.4638 3.379>
    translation=<0.246 -0.317 3.812>
   }
   n8 = "Transform (Modify) Bone - Bone: dyn_hair_tail_a_base_l"!(ComponentPose=@n7, Alpha:float=1, Translation="0,0,-1", Rotation="-10,0,0", Scale="1,1,1") -> (Pose) {
    rotationMode=BMM_Additive
    rotationSpace=BCS_BoneSpace
    translationMode=BMM_Additive
    translationSpace=BCS_BoneSpace
    alphaBoolBlend {blendOption=Linear}
    alphaScaleBiasClamp {inRange {max=0, min=-50}, outRange {max=0, min=1}}
    boneToModify {boneName=dyn_hair_tail_a_base_l}
    rotation {pitch=-10}
    translation {z=-1}
   }
   n9 = "Transform (Modify) Bone - Bone: dyn_hair_tail_a_base_r"!(ComponentPose=@n8, Alpha:float=1, Translation="0,0,0", Rotation="-10,0,0", Scale="1,1,1") -> (Pose) {
    rotationMode=BMM_Additive
    rotationSpace=BCS_BoneSpace
    translationMode=BMM_Additive
    translationSpace=BCS_BoneSpace
    alphaBoolBlend {blendOption=Linear}
    alphaScaleBiasClamp {inRange {max=0, min=-50}, outRange {max=0, min=1}}
    boneToModify {boneName=dyn_hair_tail_a_base_r}
    rotation {pitch=-10}
   }
   n10 = Get bIsPlayingEmote()
   n11 = "Constraint - dyn_hair_tail_a_base_l (PP)"!(ComponentPose=@n9, bAlphaBoolEnabled:bool=@n10, ConstraintWeights_0=0.7, ConstraintWeights_1:float=0.3) -> (Pose) {
    alphaInputType=Bool
    alphaBoolBlend {blendInTime=0.1, blendOption=Linear, blendOutTime=0.1}
    boneToModify {boneName=dyn_hair_tail_a_base_l}
    constraintSetup [{targetBone {boneName=clavicle_l}}, {targetBone {boneName=head}}]
    constraintWeights [1,1]
   }
   n12 = "Constraint - dyn_hair_tail_a_base_r (PP)"!(ComponentPose=@n11, bAlphaBoolEnabled:bool=@n10, ConstraintWeights_0=1, ConstraintWeights_1:float=0.3) -> (Pose) {
    alphaInputType=Bool
    alphaBoolBlend {blendInTime=0.1, blendOption=Linear, blendOutTime=0.1}
    boneToModify {boneName=dyn_hair_tail_a_base_r}
    constraintSetup [{targetBone {boneName=clavicle_r}}, {targetBone {boneName=head}}]
    constraintWeights [1,1]
   }
   n13 = Get HeadUp() -> (HeadUp:float)
   n14 = "Transform (Modify) Bone - Bone: dyn_hair_tail_a_base_l"(ComponentPose=@n12, Alpha=@n13, Translation="0,-4,0", Rotation="-5,-45,0", Scale="1,1,1") {
    rotationMode=BMM_Additive
    rotationSpace=BCS_BoneSpace
    translationMode=BMM_Additive
    translationSpace=BCS_BoneSpace
    alphaBoolBlend {blendOption=Linear}
    alphaScaleBiasClamp {bMapRange=true, inRange {max=0, min=-50}, outRange {max=0, min=1}}
    boneToModify {boneName=dyn_hair_tail_a_base_l}
    rotation {pitch=-5, yaw=-45}
    translation {y=-4}
   }
   n15 = "Transform (Modify) Bone - Bone: dyn_hair_tail_a_base_l"(ComponentPose=@n14, Alpha=@n13, Translation="0,0,0", Rotation="0,15,0", Scale="1,1,1") {
    rotationMode=BMM_Additive
    rotationSpace=BCS_BoneSpace
    translationMode=BMM_Additive
    translationSpace=BCS_BoneSpace
    alphaBoolBlend {blendOption=Linear}
    alphaScaleBiasClamp {bMapRange=true, inRange {max=15}}
    boneToModify {boneName=dyn_hair_tail_a_base_l}
    rotation {yaw=15}
   }
   n16 = Get HeadUp()
   n17 = "Transform (Modify) Bone - Bone: dyn_hair_tail_a_base_r"(ComponentPose=@n15, Alpha=@n16, Translation="0,4,0", Rotation="-5,-45,0", Scale="1,1,1") {
    rotationMode=BMM_Additive
    rotationSpace=BCS_BoneSpace
    translationMode=BMM_Additive
    translationSpace=BCS_BoneSpace
    alphaBoolBlend {blendOption=Linear}
    alphaScaleBiasClamp {bMapRange=true, inRange {max=0, min=-50}, outRange {max=0, min=1}}
    boneToModify {boneName=dyn_hair_tail_a_base_r}
    rotation {pitch=-5, yaw=-45}
    translation {y=4}
   }
   n18 = "Transform (Modify) Bone - Bone: dyn_hair_tail_a_base_r"(ComponentPose=@n17, Alpha=@n16, Translation="0,0,0", Rotation="0,15,0", Scale="1,1,1") {
    rotationMode=BMM_Additive
    rotationSpace=BCS_BoneSpace
    translationMode=BMM_Additive
    translationSpace=BCS_BoneSpace
    alphaBoolBlend {blendOption=Linear}
    alphaScaleBiasClamp {bMapRange=true, inRange {max=15}}
    boneToModify {boneName=dyn_hair_tail_a_base_r}
    rotation {yaw=15}
   }
   n19 = Get HeadTwist() -> (HeadTwist:float)
   n20 = "Transform (Modify) Bone - Bone: dyn_hair_tail_a_base_l"(ComponentPose=@n18, Alpha=@n19, Translation="0,0,0", Rotation="-20,0,0", Scale="1,1,1") {
    rotationMode=BMM_Additive
    rotationSpace=BCS_BoneSpace
    translationMode=BMM_Additive
    translationSpace=BCS_BoneSpace
    alphaBoolBlend {blendOption=Linear}
    alphaScaleBiasClamp {bMapRange=true, inRange {max=40}}
    boneToModify {boneName=dyn_hair_tail_a_base_l}
    rotation {pitch=-20}
   }
   n21 = "Transform (Modify) Bone - Bone: dyn_hair_tail_a_base_l"(ComponentPose=@n20, Alpha=@n19, Translation="0,-2.5,0", Rotation="20,-30,0", Scale="1,1,1") {
    rotationMode=BMM_Additive
    rotationSpace=BCS_BoneSpace
    translationMode=BMM_Additive
    translationSpace=BCS_BoneSpace
    alphaBoolBlend {blendOption=Linear}
    alphaScaleBiasClamp {bMapRange=true, inRange {max=0, min=-40}, outRange {max=0, min=1}}
    boneToModify {boneName=dyn_hair_tail_a_base_l}
    rotation {pitch=20, yaw=-30}
    translation {y=-2.5}
   }
   n22 = Get HeadTwist()
   n23 = "Transform (Modify) Bone - Bone: dyn_hair_tail_a_base_r"(ComponentPose=@n21, Alpha=@n22, Translation="0,2.5,0", Rotation="20,-25,0", Scale="1,1,1") {
    rotationMode=BMM_Additive
    rotationSpace=BCS_BoneSpace
    translationMode=BMM_Additive
    translationSpace=BCS_BoneSpace
    alphaBoolBlend {blendOption=Linear}
    alphaScaleBiasClamp {bMapRange=true, inRange {max=40}}
    boneToModify {boneName=dyn_hair_tail_a_base_r}
    rotation {pitch=20, yaw=-25}
    translation {y=2.5}
   }
   n24 = "Transform (Modify) Bone - Bone: dyn_hair_tail_a_base_r"(ComponentPose=@n23, Alpha=@n22, Translation="0,0,0", Rotation="-20,0,0", Scale="1,1,1") {
    rotationMode=BMM_Additive
    rotationSpace=BCS_BoneSpace
    translationMode=BMM_Additive
    translationSpace=BCS_BoneSpace
    alphaBoolBlend {blendOption=Linear}
    alphaScaleBiasClamp {bMapRange=true, inRange {max=0, min=-40}, outRange {max=0, min=1}}
    boneToModify {boneName=dyn_hair_tail_a_base_r}
    rotation {pitch=-20}
   }
   n25 = "Transform (Modify) Bone - Bone: dyn_helmet_base"(ComponentPose=@n24, Alpha=Get HelmetUp() -> (HelmetUp:float), Translation="0,0,0", Rotation="0,70,0", Scale="1,1,1") {
    rotationMode=BMM_Additive
    rotationSpace=BCS_BoneSpace
    translationMode=BMM_Additive
    translationSpace=BCS_BoneSpace
    alphaBoolBlend {blendOption=ExpInOut, blendOutTime=1}
    alphaScaleBiasClamp {bInterpResult=true, interpSpeedDecreasing=20, interpSpeedIncreasing=20, inRange {max=40}}
    boneToModify {boneName=dyn_helmet_base}
    rotation {yaw=70}
   }
   n26 = RigidBody!(ComponentPose=@n25, Alpha, ExternalForce, bTransferBoneVelocities:bool=False) -> (Pose) {
    bClampLinearTranslationLimitToRefPose=true
    bOverrideWorldGravity=true
    bTransferBoneVelocities=true
    overridePhysicsAsset=&/F_MED_Mechanical_Engineer_Summer_Physics
    simulationSpace=BaseBoneSpace
    alphaBoolBlend {blendOption=Linear}
    baseBoneRef {boneName=spine_05}
    componentAppliedLinearAccClamp=<3500 3500 3500>
    componentLinearAccScale=<0.1 0.1 0.1>
    componentLinearVelScale=<0.1 0.1 0.1>
    overrideWorldGravity {z=-980}
   }
   Output Pose(Result=Component To Local(ComponentPose=@n26)) {layerGroup=None}
   Get bTransferBoneVelocity() -> (bTransferBoneVelocity:bool)
  ]}
 EventGraph {nodes [
   Event BlueprintUpdateAnimation()
   n1 = Get bRigidBodyEnabled()
   n2 = IsMobilePlatform!() -> (ReturnValue:bool)
   n3 = isPcPerfMode!() -> (ReturnValue:bool)
   > Set instead_bake_pose(instead_bake_pose:bool=AND Boolean(A=OR Boolean(A=@n2, B=@n3), B=NOT Boolean(A=@n1))) -> (Output_Get:bool)
   Event BlueprintInitializeAnimation()
   > Set Body_Mesh(Body_Mesh=GetOwnerSkeletalMeshComponent())
   > n4 = Cast To PlayerPawn_Athena(Object=TryGetPawnOwner())
   > n5 = Set Pawn Owner(Pawn Owner:Player Pawn Athena Object Reference=@n4) -> (Output_Get:Player Pawn Athena Object Reference)
   > Set Head_Mesh(Head_Mesh:Skeletal Mesh Component Object Reference=GetSkeletalMeshForPartType(self=@n5, PartType=Head)) -> (Output_Get:Skeletal Mesh Component Object Reference)
   IsUsingSkinWeightProfile!(self:Skinned Mesh Component Object Reference=GetOwningComponent()) -> (ReturnValue:bool)
  ]}
]
variables [
 Body_Mesh {type=SkeletalMeshComponent}
 Head_Mesh {type=SkeletalMeshComponent}
 instead_bake_pose {type=boolean}
 Pawn Owner {type=PlayerPawn_Athena_C}
]
